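# Translucent, two-sided, unlit master material for rainbow trails on Niagara mesh particles.
# Emissive adds panning, distorted noise with RGB channel separation to a desaturated rainbow gradient, tinted by particle color.
# World position offset combines SplineThicken, noise along normals and a clamped, UV-weighted pull driven by Niagara dynamic parameters.
# Opacity uses UV gradients and particle alpha, with depth and camera fades replaced by a plain multiplier on mobile.
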Material /BRCosmetics/Effects/Niagara/MasterMaterials/MeshTrail/M_Rainbow_Mesh
graph {
 properties {
  bUsedWithNiagaraMeshParticles=true
  blendMode=BLEND_Translucent
  pixelDepthOffsetMode=PDOM_Legacy
  shadingModel=MSM_Unlit
  twoSided=true
  usedShadingModels=Unlit
 }
 nodes [
  n1 = DynamicParameter() {desc=Parameters set by Niagara, parameterIndex=1, paramNames [X,Y,Z,WPOIntensity]}
  n2 = ScalarParameter() {defaultValue=10, group=WPO, parameterName=Interpolate_Clamp}
  n3 = Clamp(
   None=Subtract(
    A=MaterialFunctionCall(X=@n1.X, Y=@n1.X, Z=@n1.X) {materialFunction=&MakeFloat3}.Result
    B=ParticlePositionWS()
   )
   Min=Multiply(A=@n2) {constB=-1}
   Max=@n2
  ) {maxDefault=1000}
  n4 = OneMinus(None=TextureCoordinate())
  n5 = Multiply(
   A=@n1.WPOIntensity
   B=Power(Base=ComponentMask(None=@n4) {g=true}, Exp=ScalarParameter() {defaultValue=3, group=WPO, parameterName=Interpolate_Power})
  )
  n6 = TextureCoordinate()
  n7 = VectorParameter() {group=WPO, parameterName=WPO UV, defaultValue=<rgba 0.5 1 0.1 0.4>}
  n8 = AppendVector(A=@n7.R, B=@n7.R)
  n9 = TextureSampleParameter2D(UVs=Panner(Coordinate=Multiply(A=@n6, B=@n8), Speed=AppendVector(A=@n7.B, B=@n7.B))) {group=WPO, parameterName=WPO Tex, samplerType=SAMPLERTYPE_Grayscale, texture=&LowResBlurredNoise}
  n10 = Multiply(A=@n9.RGB, B=VertexNormalWS())
  n11 = ScalarParameter() {defaultValue=4, group=WPO, parameterName=WPO Intensity}
  Add(
   A=MaterialFunctionCall(
    WidthBase=ScalarParameter() {defaultValue=6, group=WPO, parameterName="WidthBase (S)"}
    WidthTip=ScalarParameter() {defaultValue=20, group=WPO, parameterName="WidthTip (S)"}
   ) {materialFunction=&SplineThicken}.WorldPositionOffset
   B=Add(A=Multiply(A=@n3, B=@n5), B=Multiply(A=@n10, B=@n11))
  )
  n12 = TextureCoordinate()
  n13 = VectorParameter() {group=Distorr, parameterName=Distort UV, defaultValue=<rgba 1 1 0.02 0>}
  n14 = AppendVector(A=@n13.R, B=@n13.R)
  n15 = TextureSampleParameter2D(UVs=Panner(Coordinate=Multiply(A=@n12, B=@n14), Speed=AppendVector(A=@n13.B, B=@n13.B))) {
   group=Distorr
   parameterName=Distort Tex
   samplerType=SAMPLERTYPE_Normal
   texture=&T_Weap_Distort_N_02
  }
  n16 = TextureCoordinate()
  n17 = VectorParameter() {group=Noise, parameterName=Noise UV, channelNames {a=Pan V, b=Pan U, g=Tile V, r=Tile U}, defaultValue=<rgba 0.5 0.3 -0.05 0.2>}
  n18 = AppendVector(A=@n17.Tile U, B=@n17.Tile U)
  n19 = DynamicParameter() {paramNames [Param1, Param2, Param3, Noise Random Offset]}
  n20 = Time()
  n21 = Add(
   A=Multiply(A=ComponentMask(None=@n15.RGB) {g=true, r=true}, B=ScalarParameter() {defaultValue=0.15, group=Distorr, parameterName=Distort Intensity}) {constB=0.28}
   B=Panner(Coordinate=Multiply(A=@n16, B=@n18), Time=Add(A=@n19.Noise Random Offset, B=@n20), Speed=AppendVector(A=@n17.Pan U, B=@n17.Pan U)) {speedX=1}
  )
  n22 = ScalarParameter() {defaultValue=0.02, group=Noise, parameterName=Noise Color distortion}
  n23 = Add(A=@n21, B=@n22)
  n24 = TextureObjectParameter() {group=Noise, parameterName=Noise Tex, samplerType=SAMPLERTYPE_Masks, texture=/Game/Weapons/WeaponSkins/Wraps/Textures/Mash/Mash_Noise}
  n25 = Add(A=@n21, B=Multiply(A=@n22) {constB=-1})
  n26 = Multiply(
   A=MaterialFunctionCall(
    R=TextureSample(UVs=@n23, Tex=@n24) {texture=/Game/Effects/Fort_Effects/Textures/Plasma/T_Turbulence_Stretched}.R
    G=TextureSample(UVs=@n21, Tex=@n24) {texture=&T_Turbulence_Stretched}.G
    B=TextureSample(UVs=@n25, Tex=@n24) {texture=&T_Turbulence_Stretched}.B
   ) {materialFunction=&AppendMany}.RGB
   B=VectorParameter() {parameterName=Noise Color, defaultValue=<rgba 2 1.3333 1 1>}.RGB
  )
  n27 = Add(A=TextureCoordinate(), B=ScalarParameter() {defaultValue=0.125, group=Color, parameterName=Color Scroll})
  n28 = ScalarParameter() {defaultValue=1, group=Color, parameterName=Color Scale}
  n29 = Desaturation(
   None=TextureSample(UVs=Multiply(A=@n27, B=@n28) {constB=0.9}) {texture=/Game/Global/Textures/Default/Masks_and_Gradients/T_Grad_Rainbow}.RGB
   Fraction=ScalarParameter() {defaultValue=0.1, group=Color, parameterName=Desaturation}
  )
  n30 = ParticleColor()
  MaterialFunctionCall(Emissive=Multiply(A=Add(A=@n26, B=@n29) {constB=0}, B=@n30.RGB)) {materialFunction=&MF_FX_EmissiveAdjustments}
  n31 = ComponentMask(None=TextureCoordinate()) {r=true}
  n32 = ScalarParameter() {defaultValue=1.5, group=Opacity, parameterName=Opacity Contrast U, sortPriority=3}
  n33 = OneMinus(None=ComponentMask(None=TextureCoordinate()) {r=true})
  n34 = TextureCoordinate()
  n35 = Multiply(
   A=Multiply(A=Power(Base=@n31, Exp=@n32), B=Power(Base=@n33, Exp=@n32))
   B=Power(
    Base=ComponentMask(None=@n34) {g=true}
    Exp=ScalarParameter() {defaultValue=1.2, group=Opacity, parameterName=Opacity Contrast V, sortPriority=4}
   )
  )
  n36 = ScalarParameter() {defaultValue=2, group=Opacity, parameterName=Opacity Compensator, sortPriority=1}
  n37 = Multiply(A=Saturate(None=Multiply(A=@n35, B=@n36)), B=ScalarParameter() {defaultValue=0.5, group=Opacity, parameterName=Opacity Mul, sortPriority=1})
  n38 = Multiply(
   A=DepthFade(Opacity=@n37, FadeDistance=ScalarParameter() {defaultValue=5, group=Opacity, parameterName=Depth Fade}) {fadeDistanceDefault=5}
   B=MaterialFunctionCall(Fade Length=ScalarParameter() {defaultValue=70, group=Opacity, parameterName=Camera Fade}) {materialFunction=&CameraDepthFade}.Result
  )
  n39 = ScalarParameter() {defaultValue=4, group=Opacity, parameterName=Mobile Opacity Mul}
  Multiply(
   A=@n30.A
   B=ShadingPathSwitch(Default=@n38, Deferred=@n38, Mobile=Multiply(A=@n37, B=@n39) {constB=4}) {desc=Migrated from FeatureLevelSwitch}
  )
 ]
}
